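# Sound cue for the Candy Cane pickaxe impact, layering a random hit sound with a quieter random bell sound.
# A mixer blends one of five impact waves with one of four bell waves, the bells pitched well down.
# The mix passes through slight volume variation and a short fixed delay.
# It uses shared pickaxe impact attenuation, with overridden concurrency limiting instances per owner and stopping the oldest.
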SoundCue /BRCosmetics/Sounds/PickAxes/CandyCane/Pickaxe_CandyCane_Impact_Cue
root=Delay_1
nodes [
 {id=Delay_1, children [Modulator_0], properties {delayMax=0.1, delayMin=0.1}}
 {id=Mixer_0, children [Random_1,Modulator_1], properties {inputVolume [1,0.4]}}
 {id=Modulator_0, children [Mixer_0], properties {pitchMax=1, pitchMin=1, volumeMax=0.9, volumeMin=0.8}}
 {id=Modulator_1, children [Random_0], properties {pitchMax=0.6, pitchMin=0.6}}
 {id=Random_0, children [WavePlayer_0,WavePlayer_1,WavePlayer_2,WavePlayer_3], properties {weights [1,1,1,1]}}
 {id=Random_1, children [WavePlayer_4, WavePlayer_5, WavePlayer_6, WavePlayer_7, WavePlayer_10], properties {weights [1,1,1,1,1]}}
 {id=WavePlayer_0, properties {soundWaveAssetPtr=&/Pickaxe_CandyCane_ImpactBellsC_01}}
 {id=WavePlayer_1, properties {soundWaveAssetPtr=&/Pickaxe_CandyCane_ImpactBellsC_02}}
 {id=WavePlayer_10, properties {soundWaveAssetPtr=&/Pickaxe_CandyCane_Impact_B05}}
 {id=WavePlayer_2, properties {soundWaveAssetPtr=&/Pickaxe_CandyCane_ImpactBellsC_03}}
 {id=WavePlayer_3, properties {soundWaveAssetPtr=&/Pickaxe_CandyCane_ImpactBellsC_04}}
 {id=WavePlayer_4, properties {soundWaveAssetPtr=&/Pickaxe_CandyCane_Impact_B01}}
 {id=WavePlayer_5, properties {soundWaveAssetPtr=&/Pickaxe_CandyCane_Impact_B02}}
 {id=WavePlayer_6, properties {soundWaveAssetPtr=&/Pickaxe_CandyCane_Impact_B03}}
 {id=WavePlayer_7, properties {soundWaveAssetPtr=&/Pickaxe_CandyCane_Impact_B04}}
]
properties {
 attenuationSettings=/Game/Sounds/Attenuation/Fort_Pickaxe_Impact_Att
 audioPropertiesSheet=/BRCosmetics/Sounds/Settings/PropertySheets/Pickaxe/APS_Cosmetic_Pickaxe_Impact
 bOverrideConcurrency=true
 bPrimeOnLoad=true
 maxDistance=4300
 soundClassObject=&Fort_Impacts
 attenuationOverrides {
  bAttenuate=false
  bAttenuateWithLPF=true
  distanceAlgorithm=Logarithmic
  falloffDistance=800
  lPFRadiusMax=800
  lPFRadiusMin=400
  attenuationShapeExtents {x=300}
 }
 concurrencyOverrides {bLimitToOwner=true, maxCount=3, resolutionRule=StopOldest}
 concurrencySet [/Game/Sounds/Concurrency_Rules/OneMaximum_Owner_StopOldest]
}
